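# Animation Blueprint for the RoseScour tail accessory that reshapes the tail for skydiving states.
# AnimGraph builds four procedural poses from the cached input pose with chained, mostly additive Transform Bone nodes: default, floating, dive and dive-to-float.
# Each update, EventGraph picks the pose index from inherited skydive flags, and Blend Poses by int blends to it.
# Switching from dive to float holds the transition pose on a half-second timer, cancelled if diving resumes.
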AnimBlueprint /BRCosmetics/Accessories/FORT_Tails/Tail_F_MED_RoseScour/Meshes/F_MED_RoseScour_Tail_AnimBP
target_skeleton=/Game/Accessories/FORT_Tails/Common/Fortnite_Base_Tail_Skeleton
graphs [
 AnimGraph {nodes [
   n1 = Blend Poses by int!(
    BlendPose_0:Pose Link Structure=Use cached pose 'Default Tail Pose'()
    BlendPose_1:Pose Link Structure=Use cached pose 'SkyDiveFloating Tail Pose'()
    BlendPose_2:Pose Link Structure=Use cached pose 'SkyDiveMode Tail Pose'()
    BlendPose_3:Pose Link Structure=Use cached pose 'SkyDiveToFloat Tail Pose'()
    BlendTime_0:float=0.25
    BlendTime_1:float=0.25
    BlendTime_2:float=0.25
    BlendTime_3:float=0.25
    ActiveChildIndex:int=Get TailPoseSelection() -> (TailPoseSelection:int=0)
   ) -> (Pose:Pose Link Structure) {
    is=title AnimGraphNode_BlendListByInt
    blendPose [{linkId=-1, sourceLinkId=-1}, {linkId=-1, sourceLinkId=-1}, {linkId=-1, sourceLinkId=-1}, {linkId=-1, sourceLinkId=-1}]
    blendTime [0.1,0.1,0.1,0.1]
   }
   Output Pose(Result=@n1)
   n2 = "Transform (Modify) Bone - Bone: C_Tail_A_1_Jnt"(
    ComponentPose=Local To Component(LocalPose=Use cached pose 'Input Pose Cache'())
    Alpha=1
    Translation="0,0,0"
    Rotation="5,0,0"
    Scale="1,1,1"
   ) {rotationMode=BMM_Additive, rotationSpace=BCS_BoneSpace, scaleSpace=BCS_BoneSpace, boneToModify {boneName=C_Tail_A_1_Jnt}, rotation {pitch=5}}
   n3 = "Transform (Modify) Bone - Bone: C_Tail_A_2_Jnt"(ComponentPose=@n2, Alpha=1, Translation="0,0,0", Rotation="10,0,0", Scale="1,1,1") {
    rotationMode=BMM_Additive
    rotationSpace=BCS_BoneSpace
    scaleSpace=BCS_BoneSpace
    boneToModify {boneName=C_Tail_A_2_Jnt}
    rotation {pitch=10}
   }
   n4 = "Transform (Modify) Bone - Bone: C_Tail_A_3_Jnt"(ComponentPose=@n3, Alpha=1, Translation="0,0,0", Rotation="15,0,0", Scale="1,1,1") {
    rotationMode=BMM_Additive
    rotationSpace=BCS_BoneSpace
    scaleSpace=BCS_BoneSpace
    boneToModify {boneName=C_Tail_A_3_Jnt}
    rotation {pitch=15}
   }
   n5 = "Transform (Modify) Bone - Bone: C_Tail_A_4_Jnt"(ComponentPose=@n4, Alpha=1, Translation="0,0,0", Rotation="20,0,0", Scale="1,1,1") {
    rotationMode=BMM_Additive
    rotationSpace=BCS_BoneSpace
    scaleSpace=BCS_BoneSpace
    boneToModify {boneName=C_Tail_A_4_Jnt}
    rotation {pitch=20}
   }
   n6 = "Transform (Modify) Bone - Bone: C_Tail_A_5_Jnt"(ComponentPose=@n5, Alpha=1, Translation="0,0,0", Rotation="15,0,0", Scale="1,1,1") {
    rotationMode=BMM_Additive
    rotationSpace=BCS_BoneSpace
    scaleSpace=BCS_BoneSpace
    boneToModify {boneName=C_Tail_A_5_Jnt}
    rotation {pitch=15}
   }
   n7 = "Transform (Modify) Bone - Bone: C_Tail_A_6_Jnt"(ComponentPose=@n6, Alpha=1, Translation="0,0,0", Rotation="10,0,0", Scale="1,1,1") {
    rotationMode=BMM_Additive
    rotationSpace=BCS_BoneSpace
    scaleSpace=BCS_BoneSpace
    boneToModify {boneName=C_Tail_A_6_Jnt}
    rotation {pitch=10}
   }
   n8 = "Transform (Modify) Bone - Bone: C_Tail_A_7_Jnt"(ComponentPose=@n7, Alpha=1, Translation="0,0,0", Rotation="10,0,0", Scale="1,1,1") {
    rotationMode=BMM_Additive
    rotationSpace=BCS_BoneSpace
    scaleSpace=BCS_BoneSpace
    boneToModify {boneName=C_Tail_A_7_Jnt}
    rotation {pitch=10}
   }
   Save cached pose 'Default Tail Pose'(Pose=Component To Local(ComponentPose=@n8)) {cacheName=Default Tail Pose}
   n9 = "Transform (Modify) Bone - Bone: C_Tail_A_1_Jnt"(
    ComponentPose=Local To Component(LocalPose=Use cached pose 'Input Pose Cache'())
    Alpha=1
    Translation="0,0,0"
    Rotation="5,0,0"
    Scale="1,1,1"
   ) {
    rotationMode=BMM_Replace
    rotationSpace=BCS_BoneSpace
    scaleSpace=BCS_BoneSpace
    translationMode=BMM_Additive
    boneToModify {boneName=C_Tail_A_1_Jnt}
    rotation {pitch=5}
   }
   n10 = "Transform (Modify) Bone - Bone: C_Tail_A_2_Jnt"(ComponentPose=@n9, Alpha=1, Translation="0,0,0", Rotation="10,0,0", Scale="1,1,1") {
    rotationMode=BMM_Additive
    rotationSpace=BCS_BoneSpace
    scaleSpace=BCS_BoneSpace
    boneToModify {boneName=C_Tail_A_2_Jnt}
    rotation {pitch=10}
   }
   n11 = "Transform (Modify) Bone - Bone: C_Tail_A_3_Jnt"(ComponentPose=@n10, Alpha=1, Translation="0,0,0", Rotation="15,0,0", Scale="1,1,1") {
    rotationMode=BMM_Additive
    rotationSpace=BCS_BoneSpace
    scaleSpace=BCS_BoneSpace
    boneToModify {boneName=C_Tail_A_3_Jnt}
    rotation {pitch=15}
   }
   Save cached pose 'SkyDiveFloating Tail Pose'(Pose=Component To Local(ComponentPose=@n11)) {cacheName=SkyDiveFloating Tail Pose}
   Save cached pose 'Input Pose Cache'(Pose=Input Pose()) {cacheName=Input Pose Cache}
   n12 = "Transform (Modify) Bone - Bone: C_Tail_A_Base_Jnt"(
    ComponentPose=Local To Component(LocalPose=Use cached pose 'Input Pose Cache'())
    Alpha=1
    Translation="0,0,3"
    Rotation="0,40,0"
    Scale="1,1,1"
   ) {
    rotationMode=BMM_Additive
    rotationSpace=BCS_BoneSpace
    scaleSpace=BCS_BoneSpace
    translationMode=BMM_Additive
    boneToModify {boneName=C_Tail_A_Base_Jnt}
    rotation {yaw=40}
    translation {z=3}
   }
   Save cached pose 'SkyDiveToFloat Tail Pose'(Pose=Component To Local(ComponentPose=@n12)) {cacheName=SkyDiveToFloat Tail Pose}
   n13 = "Transform (Modify) Bone - Bone: C_Tail_A_Base_Jnt"(
    ComponentPose=Local To Component(LocalPose=Use cached pose 'Input Pose Cache'())
    Alpha=1
    Translation="0,0,3"
    Rotation="0,30,0"
    Scale="1,1,1"
   ) {
    rotationMode=BMM_Additive
    rotationSpace=BCS_BoneSpace
    scaleSpace=BCS_BoneSpace
    translationMode=BMM_Additive
    boneToModify {boneName=C_Tail_A_Base_Jnt}
    rotation {yaw=30}
    translation {z=3}
   }
   n14 = "Transform (Modify) Bone - Bone: C_Tail_A_1_Jnt"(ComponentPose=@n13, Alpha=1, Translation="0,0,0", Rotation="6,0,0", Scale="1,1,1") {rotationMode=BMM_Additive, rotationSpace=BCS_BoneSpace, scaleSpace=BCS_BoneSpace, boneToModify {boneName=C_Tail_A_1_Jnt}, rotation {pitch=6}}
   n15 = "Transform (Modify) Bone - Bone: C_Tail_A_2_Jnt"(ComponentPose=@n14, Alpha=1, Translation="0,0,0", Rotation="12,0,0", Scale="1,1,1") {
    rotationMode=BMM_Additive
    rotationSpace=BCS_BoneSpace
    scaleSpace=BCS_BoneSpace
    boneToModify {boneName=C_Tail_A_2_Jnt}
    rotation {pitch=12}
   }
   n16 = "Transform (Modify) Bone - Bone: C_Tail_A_3_Jnt"(ComponentPose=@n15, Alpha=1, Translation="0,0,0", Rotation="18,0,0", Scale="1,1,1") {
    rotationMode=BMM_Additive
    rotationSpace=BCS_BoneSpace
    scaleSpace=BCS_BoneSpace
    boneToModify {boneName=C_Tail_A_3_Jnt}
    rotation {pitch=18}
   }
   Save cached pose 'SkyDiveMode Tail Pose'(Pose=Component To Local(ComponentPose=@n16)) {cacheName=SkyDiveMode Tail Pose}
  ]}
 EventGraph {nodes [
   n1 = Reset Dive Mode to Float() {is=title K2Node_CustomEvent}
   > Set IsDiveModeToFloat(IsDiveModeToFloat:bool=false) -> (Output_Get:bool)
   Event BlueprintUpdateAnimation()
   n2 = Get bIsSkydiveFloating()
   > Set TailPoseSelection(TailPoseSelection:int=SelectInt(A=2, B=SelectInt(A=1, B=0, bPickA=@n2), bPickA=Get bIsSkydiveDiveMode())) -> (Output_Get:int=0)
   > n3 = Sequence!(execute) -> (then_0, then_1, then_2:Exec)
   > n4 = Branch(Condition=Get bIsSkydiveDiveMode())
   > Clear and Invalidate Timer by Handle(Handle=Get ReleaseTimeHandle() -> (ReleaseTimeHandle:Timer Handle Structure))
   > Set IsDiveModeToFloat(IsDiveModeToFloat=false)
   Branch(execute=@n4.else, Condition=AND Boolean(A=Get IsSkydiveDiveModePrev() -> (IsSkydiveDiveModePrev:bool), B=Get bIsSkydiveFloating()))
   > Set IsDiveModeToFloat(IsDiveModeToFloat=true)
   > n5 = Set Timer by Event(Delegate=@n1, Time=0.5, bLooping=false)
   > Set ReleaseTimeHandle(ReleaseTimeHandle:Timer Handle Structure=@n5) -> (Output_Get:Timer Handle Structure)
   Set IsSkydiveDiveModePrev(execute=@n3.then_1, IsSkydiveDiveModePrev:bool=Get bIsSkydiveDiveMode()) -> (Output_Get:bool)
   Branch(execute=@n3.then_2, Condition=Get IsDiveModeToFloat() -> (IsDiveModeToFloat:bool))
   > Set TailPoseSelection(TailPoseSelection=3) -> (Output_Get=0)
   TryGetPawnOwner()
  ]}
]
variables [
 IsDiveModeToFloat {type=boolean}
 IsSkydiveDiveModePrev {type=boolean}
 ReleaseTimeHandle {type=TimerHandle}
 TailPoseSelection {type=integer}
]
