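# Niagara glow effect for the RoseScour tail cosmetic, with two looks selected by User.VariantIndex and alpha zeroed by User.bHasCape.
# A nine-particle ribbon samples tail bones, with per-point color and waveform pulsing.
# Variant 0 is pink with pixel particles; variant 1 shades purple to orange, adding spores, pixie dust and two glow sprites on one bone.
# Loose particles chase random bone samples via moving-average interpolation, fading with distance from them.
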NiagaraSystem /BRCosmetics/Accessories/FORT_Tails/Tail_F_MED_RoseScour/VFX/NS_RoseScour_Tail_Glow
dependencies {
 usedDataInterfaces [
  &NiagaraDataInterfaceCurve
  &NiagaraDataInterfaceColorCurve
  &NiagaraDataInterfaceVector2DCurve
  &NiagaraDataInterfaceSkeletalMesh
 ]
 usedDynamicInputs [&MakeCustomLinearColorFromBool, &RandomBool]
 usedModules [
  &self:SystemSpawnScript.SystemScriptSource.SystemScriptGraph.NiagaraNodeAssignment_0.SetVariables_#1
  &SystemState
 ]
}
emitters {
 GlowDecalAndLight {
  emitterUpdateScript [
   &EmitterState(
    Inactive Response=::NewEnumerator1
    Life Cycle Mode=::NewEnumerator1
    Loop Behavior=::NewEnumerator1
    Loop Duration Mode=::NewEnumerator1
    UseLoopDelay=false
   )
   &SpawnBurst_Instantaneous(Spawn Count=2, Use Spawn Probability=false, UseLoopCountLimit=false) {module=SpawnBurst_Instantaneous001}
  ]
  particleSpawnScript [&InitializeParticle(
    Color=<rgba 5 2.5 0 0.1>
    Lifetime=0.5
    Sprite Size Mode=::NewEnumerator3
    Uniform Sprite Size=&MergeHelper_FloatSwitch(0=27.5, 1=17.5, 2=?, 3=?, Index=&ReturnExecIndex(), Num=::NewEnumerator0)
    UsePositionOffset=false
    Write Material Random=true
   ) {module=InitializeParticle001}]
  particleUpdateScript [
   &ParticleState(Kill Particles When Lifetime Has Elapsed=false)
   &SkeletalMeshLocation(
    Bone / Socket Index=8
    Bone Sampling Mode=::NewEnumerator4
    MaskSpawnGroup=false
    Sampled Position Offset=&MergeHelper_VectorSwitch(0=<2 0 0>, 1=<7.5 0 0>, 2=?, 3=?, Index=&ReturnExecIndex(), Num=::NewEnumerator0)
    Skeletal Mesh=$System.Skeletal Mesh
   )
   &ScaleSpriteSize(Uniform Curve Sprite Scale=DataInterface() {propertyValues {Curve {keys [
        {arriveTangent=1.1179, interpMode=RCIM_Cubic, leaveTangent=1.1179, tangentMode=RCTM_User, time=0, value=0.6}
        {arriveTangent=0.0224, interpMode=RCIM_Cubic, leaveTangent=0.0224, tangentMode=RCTM_User, time=1, value=1}
       ]}}})
   &CameraOffset(Camera Offset Amount=5)
   &ScaleColor(
    Scale Alpha=&FloatFromCurve(
     FloatCurve=DataInterface() {propertyValues {Curve {keys [
         {arriveTangent=4.1785, interpMode=RCIM_Cubic, leaveTangent=4.1785, tangentMode=RCTM_User, time=0, value=0}
         {arriveTangent=1.5009, interpMode=RCIM_Cubic, leaveTangent=1.5009, tangentMode=RCTM_User, time=0.2, value=0.55}
         {arriveTangent=0.0224, interpMode=RCIM_Cubic, leaveTangent=0.0224, tangentMode=RCTM_User, time=1, value=1}
        ]}}}
     Scale Curve=&Multiply_Float(
      A=&MakeCustomFloatFromBool(Bool=$User.bHasCape, True Float=0, False Float=1)
      B=&MergeHelper_FloatSwitch(
       0=0
       1=&MergeHelper_FloatSwitch(0=0.5, 1=0.25, 2=?, 3=?, Index=&ReturnExecIndex(), Num=::NewEnumerator0)
       2=?
       3=?
       Index=&MakeIntFromFloat(Float=$User.VariantIndex)
       Num=::NewEnumerator0
      )
     )
    )
    Scale RGB=&VectorFromFloat(Value=&FNTimeOfDayScalar(Night Scale=0.5, Morning Scale=1, Evening Scale=1))
    ScaleA=true
    ScaleRGB=true
   )
   &DynamicMaterialParameters(
    Index 0 Param 1=4
    Index 0 Param 2=1
    Index 0 Param 3=1
    Index 0 Param 4=1
    Module.Param0WriteEnabled=true
    Module.Param1WriteEnabled=true
    Module.Param2WriteEnabled=true
    Module.Param3WriteEnabled=true
   )
  ]
  rendererClasses [&NiagaraSpriteRendererProperties]
  renderers [{rendererClass=&NiagaraSpriteRendererProperties}]
 }
 LightSpores {
  emitterUpdateScript [
   &EmitterState(
    Inactive Response=::NewEnumerator1
    Life Cycle Mode=::NewEnumerator1
    Loop Behavior=::NewEnumerator1
    Loop Duration Mode=::NewEnumerator1
    UseLoopDelay=false
   )
   &SpawnRate(
    Spawn Probability=&MergeHelper_FloatSwitch(0=0, 1=1, 2=?, 3=?, Index=&MakeIntFromFloat(Float=$User.VariantIndex), Num=::NewEnumerator0)
    SpawnRate=12
    Use Spawn Probability=true
   ) {module=SpawnRate001}
  ]
  particleSpawnScript [
   &PartitionParticles(Number Of Partitions=5)
   &InitializeParticle(
    Color=<rgba 1 0.4667 0.2 1>
    Lifetime Max=&FN_FrontEndFloatModifier(Gameplay=2.5, Front End=1)
    Lifetime Min=2
    Lifetime Mode=::NewEnumerator1
    Sprite Size Mode=::NewEnumerator1
    Uniform Sprite Size Max=5
    Uniform Sprite Size Min=1
    UsePositionOffset=false
    Write Material Random=true
   )
  ]
  particleUpdateScript [
   &ParticleState()
   /Game/Effects/Niagara/Modules/Curves/FN_ListCurve_SkeletalMeshSampler(
    Curve Index:NiagaraFloat=&UniformAOrBFloat(
     A=&RandomRangeFloat(Minimum=0, Maximum=1)
     Evaluation Type=::NewEnumerator0
     Recalculate Random Each Loop=true
     Override Seed=false
    )
    Index_02=1
    Index_03=2
    Index_04=3
    Index_05=4
    Index_06=5
    Index_07=6
    Index_08=7
    Index_09=8
    "Interpolate Mesh Sampling (Update)"=true
    Number of Inputs=7
    Offset Coordinate Space=3
    Offset_01:Vector3f=<0 0 1>
    Offset_02:Vector3f=<0 -1 0>
    Offset_03:Vector3f=<0 -1 0>
    Offset_04:Vector3f=<0 -1 0>
    Offset_05:Vector3f=<0 -1 0>
    Offset_06:Vector3f=<0 -1 0>
    Offset_07:Vector3f=<0 -1 0>
    Offset_08:Vector3f=<0 -1 0>
    Offset_09:Vector3f=<0 -1 0>
    Skeletal Mesh:NiagaraDataInterfaceSkeletalMesh=$System.Skeletal Mesh
   )
   &InterpolateOverTime(Rate Of Change=0.35, Target Position Value:NiagaraPosition=$Output.FN_ListCurve_SkeletalMeshSampler.Position, Variable Type=::NewEnumerator6)
   /SetVariables(Particles.Position=$Particles.InterpolateOverTime.InterpolateOverTime_Position.Moving Average) {isSetParameters=true}
   &ScaleColor(
    Scale Alpha=&FloatFromCurve(
     FloatCurve=DataInterface() {propertyValues {Curve {
        defaultValue=1
        keys [
         {arriveTangent=-9.5851, interpMode=RCIM_Cubic, leaveTangent=-9.5851, tangentMode=RCTM_User, time=0, value=1}
         {arriveTangent=-0.2671, interpMode=RCIM_Cubic, leaveTangent=-0.2671, tangentMode=RCTM_User, time=0.1509, value=0.1815}
         {interpMode=RCIM_Cubic, tangentMode=RCTM_User, time=1, value=0}
        ]
       }}}
     CurveIndex=&NormalizeDistanceRange(
      Start Position:NiagaraPosition=$Particles.Position
      End Position:NiagaraPosition=$Output.FN_ListCurve_SkeletalMeshSampler.Position
      Distance:NiagaraFloat=100
      Invert Normalized Range:NiagaraBool=false
     )
    )
    ScaleA=true
    ScaleRGB=true
   )
   &CameraOffset(Camera Offset Amount=&RandomRangeFloat(Minimum=0, Maximum=10))
   &GravityForce(Gravity=&RandomRangeVector(Minimum=<0 0 33>, Maximum=<0 0 99>))
   &SolveForcesAndVelocity(Clamp Velocity=false, Limit Acceleration=false)
   &ScaleSpriteSize(
    Uniform Curve Scale=&Waveform(
     "Waveform [1]"=::NewEnumerator0
     "[1] Frequency"=&RandomRangeFloat(Minimum=0, Maximum=1.8831)
     "[1] Amplitude Min/Max"=<0.7 1>
    )
    Uniform Curve Sprite Scale=DataInterface() {propertyValues {Curve {keys [
        {interpMode=RCIM_Cubic, tangentMode=RCTM_User, time=0, value=1}
        <at 0.9 1>
        {arriveTangent=-4.3352, interpMode=RCIM_Cubic, leaveTangent=-4.3352, tangentMode=RCTM_User, time=1, value=0}
       ]}}}
   ) {module=ScaleSpriteSize001}
   &ScaleColor(
    Scale Alpha=&FloatFromCurve(
     FloatCurve=DataInterface() {propertyValues {Curve {
        defaultValue=1
        keys [
         {interpMode=RCIM_Cubic, tangentMode=RCTM_User, time=0, value=0}
         {interpMode=RCIM_Cubic, tangentMode=RCTM_User, time=0.5, value=1}
         {tangentMode=RCTM_User, time=1, value=0}
        ]
       }}}
     Scale Curve=&MakeCustomFloatFromBool(Bool=$User.bHasCape, True Float=0, False Float=1)
    )
    ScaleA=true
    ScaleRGB=true
   ) {module=ScaleColor001}
  ]
  rendererClasses [&NiagaraSpriteRendererProperties]
  renderers [
   {rendererClass=&NiagaraSpriteRendererProperties}
   {rendererClass=&NiagaraSpriteRendererProperties, rendererIndex=1}
  ]
 }
 Pixel_Particles {
  emitterUpdateScript [
   &EmitterState()
   &SpawnRate(
    Spawn Probability=&MergeHelper_FloatSwitch(0=1, 1=0, 2=?, 3=?, Index=&MakeIntFromFloat(Float=$User.VariantIndex), Num=::NewEnumerator0)
    SpawnRate=&RandomRangeFloat(Minimum=5, Maximum=10)
    Use Spawn Probability=true
   )
  ]
  particleSpawnScript [
   &PartitionParticles(Number Of Partitions=5)
   &InitializeParticle(
    Color=<rgba 2 1 1.201 1>
    Lifetime Max=0.5
    Lifetime Min=0.35
    Lifetime Mode=::NewEnumerator1
    Sprite Size Mode=::NewEnumerator1
    Uniform Sprite Size Max=5
    Uniform Sprite Size Min=2.5
    UsePositionOffset=false
    Write Material Random=true
   )
   &CameraOffset(Camera Offset Amount=-25)
  ]
  particleUpdateScript [
   &ParticleState()
   &FN_ListCurve_SkeletalMeshSampler(
    Curve Index:NiagaraFloat=&RandomRangeFloat(Minimum=0, Maximum=1)
    Index_02=1
    Index_03=2
    Index_04=3
    Index_05=4
    Index_06=5
    Index_07=6
    Index_08=7
    Index_09=8
    "Interpolate Mesh Sampling (Update)"=true
    Number of Inputs=7
    Offset Coordinate Space=3
    Offset_01:Vector3f=<0 0 1>
    Offset_02:Vector3f=<0 -1 0>
    Offset_03:Vector3f=<0 -1 0>
    Offset_04:Vector3f=<0 -1 0>
    Offset_05:Vector3f=<0 -1 0>
    Offset_06:Vector3f=<0 -1 0>
    Offset_07:Vector3f=<0 -1 0>
    Offset_08:Vector3f=<0 -1 0>
    Offset_09:Vector3f=<0 -1 0>
    Skeletal Mesh:NiagaraDataInterfaceSkeletalMesh=$System.Skeletal Mesh
   )
   &InterpolateOverTime(Target Position Value:NiagaraPosition=$Output.FN_ListCurve_SkeletalMeshSampler.Position, Variable Type=::NewEnumerator6)
   /SetVariables(Particles.Position=$Particles.InterpolateOverTime.InterpolateOverTime_Position.Moving Average) {isSetParameters=true}
   &ScaleColor(
    Scale Alpha=&FloatFromCurve(
     FloatCurve=DataInterface() {propertyValues {Curve {
        defaultValue=1
        keys [
         {arriveTangent=-15.6679, interpMode=RCIM_Cubic, leaveTangent=-15.6679, tangentMode=RCTM_User, time=0, value=1}
         {arriveTangent=-0.2671, interpMode=RCIM_Cubic, leaveTangent=-0.2671, tangentMode=RCTM_User, time=0.1223, value=0.0874}
         {interpMode=RCIM_Cubic, tangentMode=RCTM_User, time=1, value=0}
        ]
       }}}
     CurveIndex=&NormalizeDistanceRange(
      Start Position:NiagaraPosition=$Particles.Position
      End Position:NiagaraPosition=$Output.FN_ListCurve_SkeletalMeshSampler.Position
      Distance:NiagaraFloat=50
      Invert Normalized Range:NiagaraBool=false
     )
    )
    ScaleA=true
    ScaleRGB=true
   ) {module=ScaleColor001}
   &Drag(Drag=2, Use Linear Drag=true, Use Rotational Drag=true)
   &ScaleSpriteSize(
    Uniform Curve Scale=0.6
    Uniform Curve Sprite Scale=DataInterface() {propertyValues {Curve {keys [
        {arriveTangent=8.9937, interpMode=RCIM_Cubic, leaveTangent=8.9937, tangentMode=RCTM_User, time=0, value=0}
        {arriveTangentWeight=0.1601, interpMode=RCIM_Cubic, leaveTangentWeight=0.2668, tangentMode=RCTM_User, time=0.2, value=1}
        {arriveTangentWeight=0.2906, interpMode=RCIM_Cubic, leaveTangentWeight=0.1163, tangentMode=RCTM_User, time=0.7, value=1}
        {arriveTangent=-6.7083, interpMode=RCIM_Cubic, leaveTangent=-6.7083, tangentMode=RCTM_User, time=1, value=0}
       ]}}}
   )
   &ScaleColor(
    Scale Alpha=&FloatFromCurve(
     FloatCurve=DataInterface() {propertyValues {Curve {keys [<at 0.8 1>,<at 1 0>]}}}
     Scale Curve=&MakeCustomFloatFromBool(Bool=$User.bHasCape, True Float=0, False Float=1)
    )
    ScaleA=true
    ScaleRGB=true
   )
   &SolveForcesAndVelocity(Clamp Velocity=false, Limit Acceleration=false)
   &DynamicMaterialParameters(
    Index 0 Param 1=2
    Module.Index0Param0Write=true
    Module.Index0Param1Write=true
    Module.Index0Param2Write=true
    Module.Index0Param3Write=true
   )
  ]
  rendererClasses [&NiagaraSpriteRendererProperties]
  renderers [{rendererClass=&NiagaraSpriteRendererProperties}]
 }
 PixieDust {
  emitterUpdateScript [
   &EmitterState(
    Inactive Response=::NewEnumerator1
    Life Cycle Mode=::NewEnumerator1
    Loop Behavior=::NewEnumerator1
    Loop Duration Mode=::NewEnumerator1
    UseLoopDelay=false
   )
   &SpawnRate(
    Spawn Probability=&MergeHelper_FloatSwitch(0=0, 1=1, 2=?, 3=?, Index=&MakeIntFromFloat(Float=$User.VariantIndex), Num=::NewEnumerator0)
    SpawnRate=12
    Use Spawn Probability=true
   )
  ]
  particleSpawnScript [
   &PartitionParticles(Number Of Partitions=5)
   &InitializeParticle(
    Color Maximum=<rgba 0.808 0.4909 0.1856 0.75>
    Color Minimum=<rgba 0.9445 0.5331 0.1371 0.75>
    Color Mode=::NewEnumerator2
    Lifetime Max=1
    Lifetime Min=0.5
    Lifetime Mode=::NewEnumerator1
    Sprite Rotation Mode=::NewEnumerator2
    Sprite Size Mode=::NewEnumerator1
    Uniform Sprite Size Max=30
    Uniform Sprite Size Min=15
    UsePositionOffset=false
    Write Material Random=true
   )
   &DynamicMaterialParameters(
    Index 0 Param 1=&RandomRangeFloat(Minimum=-0.025, Maximum=0.025)
    Index 0 Param 2=&RandomRangeFloat(Minimum=0, Maximum=32)
    Module.Index0Param0Write=true
    Module.Index0Param1Write=true
    Module.Index0Param2Write=false
    Module.Index0Param3Write=false
   )
   &CameraOffset(Camera Offset Amount=25)
  ]
  particleUpdateScript [
   &ParticleState()
   &FN_ListCurve_SkeletalMeshSampler(
    Curve Index:NiagaraFloat=&UniformAOrBFloat(
     A=&RandomRangeFloat(Minimum=0, Maximum=1)
     Evaluation Type=::NewEnumerator0
     Recalculate Random Each Loop=true
     Override Seed=false
    )
    Index_02=1
    Index_03=2
    Index_04=3
    Index_05=4
    Index_06=5
    Index_07=6
    Index_08=7
    Index_09=8
    "Interpolate Mesh Sampling (Update)"=true
    Number of Inputs=7
    Offset Coordinate Space=3
    Offset_01:Vector3f=<0 0 1>
    Offset_02:Vector3f=<0 -1 0>
    Offset_03:Vector3f=<0 -1 0>
    Offset_04:Vector3f=<0 -1 0>
    Offset_05:Vector3f=<0 -1 0>
    Offset_06:Vector3f=<0 -1 0>
    Offset_07:Vector3f=<0 -1 0>
    Offset_08:Vector3f=<0 -1 0>
    Offset_09:Vector3f=<0 -1 0>
    Skeletal Mesh:NiagaraDataInterfaceSkeletalMesh=$System.Skeletal Mesh
   )
   &InterpolateOverTime(Rate Of Change=0.35, Target Position Value:NiagaraPosition=$Output.FN_ListCurve_SkeletalMeshSampler.Position, Variable Type=::NewEnumerator6)
   /SetVariables(Particles.Position=$Particles.InterpolateOverTime.InterpolateOverTime_Position.Moving Average) {isSetParameters=true}
   &ScaleColor(
    Scale Alpha=&FloatFromCurve(
     FloatCurve=DataInterface() {propertyValues {Curve {
        defaultValue=1
        keys [
         {arriveTangent=-9.5851, interpMode=RCIM_Cubic, leaveTangent=-9.5851, tangentMode=RCTM_User, time=0, value=1}
         {arriveTangent=-0.2671, interpMode=RCIM_Cubic, leaveTangent=-0.2671, tangentMode=RCTM_User, time=0.1509, value=0.1815}
         {interpMode=RCIM_Cubic, tangentMode=RCTM_User, time=1, value=0}
        ]
       }}}
     CurveIndex=&NormalizeDistanceRange(
      Start Position:NiagaraPosition=$Particles.Position
      End Position:NiagaraPosition=$Output.FN_ListCurve_SkeletalMeshSampler.Position
      Distance:NiagaraFloat=100
      Invert Normalized Range:NiagaraBool=false
     )
    )
    ScaleA=true
    ScaleRGB=true
   ) {module=ScaleColor001}
   &ScaleColor(
    Scale Alpha=&FloatFromCurve(
     FloatCurve=DataInterface() {propertyValues {Curve {
        defaultValue=1
        keys [
         {interpMode=RCIM_Cubic, tangentMode=RCTM_User, time=0, value=0}
         {interpMode=RCIM_Cubic, tangentMode=RCTM_User, time=0.5, value=1}
         {tangentMode=RCTM_User, time=1, value=0}
        ]
       }}}
     Scale Curve=&MakeCustomFloatFromBool(Bool=$User.bHasCape, True Float=0, False Float=1)
    )
    ScaleA=true
    ScaleRGB=true
   )
   &ScaleSpriteSize(Uniform Curve Sprite Scale=DataInterface() {propertyValues {Curve {keys [
        {arriveTangent=1.3913, interpMode=RCIM_Cubic, leaveTangent=1.3913, tangentMode=RCTM_User, time=0, value=0.75}
        {tangentMode=RCTM_User, time=1, value=1.25}
       ]}}})
   &GravityForce(Gravity=&RandomRangeVector(Minimum=<0 0 33>, Maximum=<0 0 99>))
   &SolveForcesAndVelocity(Clamp Velocity=false, Limit Acceleration=false)
  ]
  rendererClasses [&NiagaraSpriteRendererProperties]
  renderers [{rendererClass=&NiagaraSpriteRendererProperties}]
 }
 Trail_Tail {
  emitterUpdateScript [
   &EmitterState(
    Inactive Response=::NewEnumerator1
    Life Cycle Mode=::NewEnumerator1
    Loop Behavior=::NewEnumerator1
    Loop Duration Mode=::NewEnumerator1
    UseLoopDelay=false
   )
   &SpawnBurst_Instantaneous(Spawn Count=9, Use Spawn Probability=false, UseLoopCountLimit=false)
  ]
  particleSpawnScript [
   &InitializeParticle(
    Color=<rgba 1 1 1 0.33>
    Lifetime=5
    Ribbon Width=7.5
    Ribbon Width Mode=::NewEnumerator1
    UsePositionOffset=false
    Write Material Random=true
   )
   &PartitionParticles(
    Normalized Index In Partition Includes One=true
    Normalized Index Includes One=true
    Number Of Partitions=9
    Particles Per Partition=1
    Partition Mode:ENiagara_PartitionMode_v2=::NewEnumerator1
    Write Partition Index to Ribbon Link Order:ENiagara_AttributeSamplingApplyOutput=::NewEnumerator0
   )
  ]
  particleUpdateScript [
   &ParticleState(
    Kill Particles When Lifetime Has Elapsed=false
    Let Infinitely Lived Particles Die When Emitter Deactivates=true
    Loop Particles Lifetime=true
   )
   &FN_ListCurve_SkeletalMeshSampler(
    Apply Orientation=true
    Apply Position=true
    Curve Index:NiagaraFloat=$Particles.PartitionParticles.NormalizedPartition
    Index_02=1
    Index_03=2
    Index_04=3
    Index_05=4
    Index_06=5
    Index_07=6
    Index_08=7
    Index_09=8
    Number of Inputs=7
    Offset Coordinate Space=3
    Offset_01:Vector3f=<0 0 1>
    Offset_02:Vector3f=<0 -1 0>
    Offset_03:Vector3f=<0 -1 -0.5>
    Offset_04:Vector3f=<0 -1 -2>
    Offset_05:Vector3f=<0 -1 -2>
    Offset_06:Vector3f=<0 -1 -2>
    Offset_07:Vector3f=<0 -1 -2>
    Offset_08:Vector3f=<0 -1 -1>
    Offset_09:Vector3f=<0 -1 0>
    Skeletal Mesh:NiagaraDataInterfaceSkeletalMesh=$System.Skeletal Mesh
   )
   &ScaleRibbonWidth(Ribbon Width Scale=&FloatFromCurve(
     FloatCurve=DataInterface() {propertyValues {Curve {
        defaultValue=1
        keys [
         {arriveTangent=2, time=0, value=1}
         {interpMode=RCIM_Cubic, tangentMode=RCTM_User, time=0.5, value=1}
         {arriveTangent=-2.3585, tangentMode=RCTM_User, time=1, value=0.5}
        ]
       }}}
     CurveIndex=$Particles.PartitionParticles.NormalizedPartition
    ))
   /Niagara/Noise/Waveform(
    EnableExponent1=false
    GlobalClampMax=false
    GlobalClampMin=false
    UseInput1=true
    UsePhase1=false
    "[1] Frequency"=0.33
    "[1] Input Position":NiagaraFloat=&OneMinusFloat(Float=&ReturnNormalizedExecIndex(Normalized Index Scale=2, Normalized Index Includes One=false))
   )
   /Game/Effects/Niagara/Modules/Curves/FN_ListCurve(
    Curve Index:NiagaraFloat=$Particles.PartitionParticles.NormalizedPartition
    Input Type=4
    Linear Color_01:LinearColor=&MergeHelper_ColorSwitch(
     0=<rgba 1 0 0.5833 0.33>
     1=<rgba 0.375 0.25 1 0.33>
     2=?
     3=?
     Index=&MakeIntFromFloat(Float=$User.VariantIndex)
     Num=::NewEnumerator0
    )
    Linear Color_02:LinearColor=&MergeHelper_ColorSwitch(
     0=<rgba 1 0 0.5833 0.33>
     1=<rgba 0.375 0.25 1 0.33>
     2=?
     3=?
     Index=&MakeIntFromFloat(Float=$User.VariantIndex)
     Num=::NewEnumerator0
    )
    Linear Color_03:LinearColor=&MergeHelper_ColorSwitch(
     0=<rgba 1 0 0.5833 0.33>
     1=<rgba 0.375 0.25 1 0.33>
     2=?
     3=?
     Index=&MakeIntFromFloat(Float=$User.VariantIndex)
     Num=::NewEnumerator0
    )
    Linear Color_04:LinearColor=&MergeHelper_ColorSwitch(
     0=<rgba 1 0 0.5833 0.33>
     1=<rgba 0.375 0.25 1 0.33>
     2=?
     3=?
     Index=&MakeIntFromFloat(Float=$User.VariantIndex)
     Num=::NewEnumerator0
    )
    Linear Color_05:LinearColor=&MergeHelper_ColorSwitch(
     0=<rgba 1 0 0.5833 0.33>
     1=<rgba 0.375 0.25 1 0.33>
     2=?
     3=?
     Index=&MakeIntFromFloat(Float=$User.VariantIndex)
     Num=::NewEnumerator0
    )
    Linear Color_06:LinearColor=&MergeHelper_ColorSwitch(
     0=<rgba 1 0 0.5833 0.33>
     1=<rgba 0.375 0.25 1 0.33>
     2=?
     3=?
     Index=&MakeIntFromFloat(Float=$User.VariantIndex)
     Num=::NewEnumerator0
    )
    Linear Color_07:LinearColor=&MergeHelper_ColorSwitch(
     0=<rgba 1 0 0.5833 0.33>
     1=<rgba 0.7136 0.4364 0.7304 0.33>
     2=?
     3=?
     Index=&MakeIntFromFloat(Float=$User.VariantIndex)
     Num=::NewEnumerator0
    )
    Linear Color_08:LinearColor=&MergeHelper_ColorSwitch(
     0=<rgba 1 0 0.5833 0.33>
     1=<rgba 0.7304 0.4743 0.4181 0.33>
     2=?
     3=?
     Index=&MakeIntFromFloat(Float=$User.VariantIndex)
     Num=::NewEnumerator0
    )
    Linear Color_09:LinearColor=&MergeHelper_ColorSwitch(
     0=<rgba 1 0 0.5833 0.33>
     1=<rgba 1 0.6022 0.25 0.33>
     2=?
     3=?
     Index=&MakeIntFromFloat(Float=$User.VariantIndex)
     Num=::NewEnumerator0
    )
    Number Inputs Switch=7
   )
   &ScaleColor(
    Scale Alpha=&Multiply_Float(
     A=&FloatFromCurve(
      FloatCurve=DataInterface() {propertyValues {Curve {
         defaultValue=1
         keys [
          {interpMode=RCIM_Cubic, tangentMode=RCTM_User, time=0, value=1}
          {arriveTangent=-0.9914, interpMode=RCIM_Cubic, leaveTangent=-0.9914, tangentMode=RCTM_User, time=1, value=0.5}
         ]
        }}}
      CurveIndex=$Particles.PartitionParticles.NormalizedPartition
      Scale Curve=&MakeCustomFloatFromBool(Bool=$User.bHasCape, True Float=0, False Float=1)
     )
     B=&Add_Float(A=0.5, B=&Lerp_Float(A=0, B=0.25, Alpha=$Particles.Waveform.Waveform))
    )
    Scale RGB=&Multiply_Vector(
     A=&MakeVectorFromLinearColorRGB(Color=$Output.FN_ListCurve.Linear Color)
     B=&Lerp_Vector(A=<1 1 1>, B=<1.5 1.5 1.5>, Alpha=$Particles.Waveform.Waveform)
    )
    ScaleA=true
    ScaleRGB=true
   )
  ]
  rendererClasses [&NiagaraRibbonRendererProperties]
  renderers [{rendererClass=&NiagaraRibbonRendererProperties}]
 }
}
summary {
 systemName=NS_RoseScour_Tail_Glow
 emitters [
  {emitterName=Trail_Tail, rendererClasses [&NiagaraRibbonRendererProperties]}
  {emitterName=Pixel_Particles, rendererClasses [&NiagaraSpriteRendererProperties]}
  {emitterName=LightSpores, rendererClasses [&NiagaraSpriteRendererProperties]}
  {emitterName=PixieDust, rendererClasses [&NiagaraSpriteRendererProperties]}
  {emitterName=GlowDecalAndLight, rendererClasses [&NiagaraSpriteRendererProperties]}
 ]
 userVariables [
  User.bHasCape {type=NiagaraBool}
  User.bHasJacket {type=NiagaraFloat, defaultValue=0, description=""}
  User.bisFootHidden {type=NiagaraBool}
  User.bIsFrontEnd {type=NiagaraBool}
  User.bIsFrontEndPreview {type=NiagaraBool}
  User.bisHandHidden {type=NiagaraBool}
  User.bisHeadHidden {type=NiagaraBool}
  User.bShowBackFX {type=NiagaraBool}
  User.VariantIndex {type=NiagaraFloat, defaultValue=0, description=""}
 ]
}
